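# EQS query that finds a location for a new earthquake, judging by its name, by picking among earthquake spawner actors.
# A custom Blueprint generator supplies the valid spawner actors as candidates.
# A 2D distance test against the all-earthquakes context discards spawners under a minimum distance and linearly scores farther ones higher.
# The query itself has no random test, so any randomness comes from outside this asset.
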EnvQuery /Milkshake/Gameplay/EQS/EQS_Milkshake_RandomEarthquakeLocationFinder
options [{
  generator {
   name=EnvQueryGenerator_Milkshake_ValidEarthquakeSpawners_C_0
   class=EnvQueryGenerator_Milkshake_ValidEarthquakeSpawners_C
   properties {
    bAutoSortTests=true
    bCanRunAsync=false
    context=&EnvQueryContext_Querier
    envQueryResultNormalizationOption=Default
    generatedItemType=/Script/AIModule.EnvQueryItemType_Actor
    generatorsActionDescription=""
    optionName=""
   }
  }
  tests [EnvQueryTest_Distance_1 {
    class=EnvQueryTest_Distance
    properties {
     distanceTo=&/EnvQueryContext_Milkshake_AllEarthquakes.EnvQueryContext_Milkshake_AllEarthquakes_C
     filterType=Minimum
     floatValueMin=32500
     scoringEquation=Linear
     scoringFactor=1
     testComment=""
     testMode=Distance2D
     testPurpose=FilterAndScore
    }
   }]
 }]
