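; EQS query context for the Milkshake mode that supplies every earthquake actor in the world.
; ProvideActorsSet calls GetAllActorsOfClass for BP_Milkshake_Earthquake and returns the results, ignoring the querier inputs.
; Environment queries can then test or generate around all active earthquakes.
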(fn ProvideActorsSet
  (params (QuerierObject :type Object%20Reference) (QuerierActor :type Actor%20Object%20Reference))
  ;[0.07629 0 0.1771] Return all the earthquake actors currently in the world.
  (bind _outactors (Actor|GetAllActorsOfClass "/Milkshake/Gameplay/BP_Milkshake_Earthquake.BP_Milkshake_Earthquake_C"))
  (return _outactors))
