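; Blueprint EQS generator that supplies earthquake spawner actors currently able to spawn an earthquake as query items.
; It gathers every BP_Milkshake_RandomEarthquakeSpawner in the world and keeps valid ones whose CanSpawnEarthquake check passes.
; Passing spawners are added as generated actors; the context locations input is ignored, so results are not filtered by proximity.
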(fn DoItemGeneration
  (params (ContextLocations :type Array%20of%20Vectors))
  (bind _outactors (Actor|GetAllActorsOfClass "/Milkshake/Gameplay/BP_Milkshake_RandomEarthquakeSpawner.BP_Milkshake_RandomEarthquakeSpawner_C"))
  (for _array_element _outactors
    (Utilities|IsValid _array_element
      (:Is%20Valid
        (bind _bcanspawn (Class|BPMilkshakeRandomEarthquakeSpawner|CanSpawnEarthquake _array_element))
        (if _bcanspawn
          (EQS|AddGeneratedActor _array_element)))))
  (return))
